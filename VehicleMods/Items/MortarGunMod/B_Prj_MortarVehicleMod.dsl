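; Projectile Blueprint for the mortar gun vehicle mod, which explodes on bouncing into player pawns or vehicles other than the firing player's.
; Entering water calls DelayBeforeExplode and sets a flag that later selects the water explosion sound.
; Exploding deactivates the Niagara trail.
; On disable for kill, cosmetic-only logic plays the explosion sound and fades out the looping tell and water fuse audio.
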(event EventBeginPlay
  (|CallParentFunction)
  (Water|AssignOnEnterWaterEvent (Variables|Water|GetWaterInteractionComponent) (delegate Custom|OnPrjEnterWaterEvent)))

(event Custom|OnPrjEnterWaterEvent
  (params (WaterBody :type Fort%20Water%20Body%20Actor%20Object%20Reference) (WaterInteractionComponent :type Fort%20Water%20Interaction%20Component%20Object%20Reference) (bIsFirstBody :type Boolean))
  (CallFunction|DelayBeforeExplode)
  (Variables|Default|SetbIsInWater true))

(event EventOnBounce
  (params (Hit :type Hit%20Result%20Structure%20%28by%20ref%29))
  (bind (_bblockinghit _binitialoverlap _time _distance _location _impactpoint _normal _impactnormal _physmat _hitactor) (Collision|BreakHitResult Hit))
  ;[1 1 1] If we hit a vehicle or a Pawn, explode!
  (|CallParentFunction Hit)
  (bind _asfort_athena_vehicle (Utilities|Casting|CastToFortAthenaVehicle _hitactor)
    (:then
      (bind _returnvalue (CallFunction|GetOwnerWeapon))
      (if (!= _asfort_athena_vehicle (Vehicle|GetVehicleActor (Utilities|Casting|CastToFortPlayerPawn (Actor|GetOwner _returnvalue)) :!class /Script/FortniteGame.FortPlayerPawn))
        _))
    (:CastFailed
      (Utilities|Casting|CastToFortPlayerPawn _hitactor)))
  (Delivery|DoExplosion))

(event EventOnExploded
  (params (HitActors :type Array%20of%20Actor%20Object%20References) (HitResults :type Array%20of%20Hit%20Result%20Structures))
  (Components|Activation|Deactivate (Variables|Default|GetNS_MortarVehicle_Trail)))

(event EventOnDisableForKill
  (|CallParentFunction)
  (Networking|CanExecuteCosmeticEvents
    (:True
      (SoundLibrary|PlaySoundSimple (WorldSubsystems|GetSoundLibrarySubsystem) self (Utilities|Struct|MakeSimpleSoundSettings (select (Variables|Default|GetbIsInWater) (Variables|Default|GetWaterExplosionSoundTag) (Variables|Default|GetExplosionSoundTag))) :!class /Script/SoundLibrary.SoundLibrarySubsystem)
      (bind _loopingaudiotell (Variables|Default|GetLoopingAudioTell)
        (:then
          (Audio|Components|Audio|FadeOut _loopingaudiotell 0.3 0.0 "Linear" :!class /Script/Engine.AudioComponent))
        (:else
          _))
      (bind _waterfuseaudiocomponent (Variables|Default|GetWaterFuseAudioComponent)
        (:then
          (Audio|Components|Audio|FadeOut _waterfuseaudiocomponent 0.3 0.0 "Linear" :!class /Script/Engine.AudioComponent))
        (:else)))))

(fn ConstructionScript
  (|CallParentFunction))
