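; Vehicle-mounted mortar weapon Blueprint for the vehicle mod, with self-refilling ammo and equip and fire feedback.
; BeginPlay reads refill interval and bullets per load from scalable float rows and binds to ammo consumption.
; Consuming ammo starts a looping timer, unless one is already running, that grants ammo until HasFullClip clears it.
; Equip and unequip execute GameplayCues on the instigator, and firing adds a camera shake for its local viewing player.
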(event EventBeginPlay
  ;[1 1 1] BeginPlay
  (EventDispatchers|AssignOnConsumeAmmo (delegate Custom|OnConsumeAmmo_Event))
  (Variables|Default|SetRefillAmmoTimer (ScalableFloat|GetValueatLevel (Variables|Default|GetRow_AmmoRefillRate)))
  (Variables|Default|SetNumberOfBulletsPerLoad (ScalableFloat|GetValueasInteger (Variables|Default|GetRowNumberofBulletsPerLoad))))

(event Custom|OnConsumeAmmo_Event
  (params (Quantity :type Integer))
  ;[1 1 1] Local Ammo Changed
  (if (not (Utilities|Time|IsValidTimerHandle (Variables|Default|GetGrantAmmoTimer)))
    (bind _returnvalue (Utilities|Time|SetTimerbyEvent (EventDispatchers|CreateEvent :!event Custom|GrantAmmo) (Variables|Default|GetRefillAmmoTimer) true))
    (Variables|Default|SetGrantAmmoTimer _returnvalue)))

(event Custom|GrantAmmo
  ;[1 1 1] Grant Ammo to player
  (FortWeapon|ModAmmo (Variables|Default|GetNumberOfBulletsPerLoad)) ; Grant additional Ammo to Clip
  (if (CallFunction|HasFullClip)
    (Utilities|Time|ClearandInvalidateTimerbyHandle (Variables|Default|GetGrantAmmoTimer))))

(event FortWeapon|EventOnEquippedAndAttached
  ;[1 1 1] On Equipped and Attached
  (GameplayCue|ExecuteGameplayCue (Game|GetInstigator) (Variables|Default|GetGC_WeaponEquipped)))

(event EventOnPlayWeaponFireFX
  (params (bPersistentFire :type Boolean) (bSecondaryFire :type Boolean))
  ;[1 1 1] Camera Shake on Weapon Fire
  (|CallParentFunction bPersistentFire bSecondaryFire)
  (Game|Feedback|ClientStartCameraShake :self (Pawn|GetLocalViewingPlayerController (Game|GetInstigator)) :Shake (Variables|Default|GetWeaponFireCameraShake) :Scale 1.0 :PlaySpace "CameraLocal"))

(event FortWeapon|EventOnUnEquippedAndDetached
  ;[1 1 1] On Equipped and Attached
  (GameplayCue|ExecuteGameplayCue (Game|GetInstigator) (Variables|Default|GetGC_WeaponUnequip)))

(fn HasFullClip
  (return (>= (FortWeapon|GetLocalMagazineAmmoCount) (FortWeapon|GetBulletsPerClip))))

(fn ConstructionScript)
